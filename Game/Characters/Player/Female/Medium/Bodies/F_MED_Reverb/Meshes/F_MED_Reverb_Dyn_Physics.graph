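# Physics asset for the dynamic coat and string bones on the medium female Reverb player body.
# Two spine capsules form the torso, with five boxed coat panels joined to the upper one by limited swing and twist.
# Angular locks tie the back panel to both side panels.
# Left and right three-segment string chains hang from free-rotating base capsules on the upper spine, with limited swing and almost no twist between segments.
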PhysicsAsset /Game/Characters/Player/Female/Medium/Bodies/F_MED_Reverb/Meshes/F_MED_Reverb_Dyn_Physics
bodies {
 dyn_coat_bk [{shapeType=Box, extentX=2.7215, extentY=5.7989, extentZ=13.811, center=<-0.5286 0 -3.1424>, rotation=<rot 0 180 -90>}]
 dyn_coat_fr_l [{shapeType=Box, extentX=2.1237, extentY=5.5238, extentZ=10.0867, center=<-0.6809 1.5738 -3.9294>, rotation=<rot 0 -169.9245 -90.6937>}]
 dyn_coat_fr_r [{shapeType=Box, extentX=2.1237, extentY=5.0238, extentZ=10.0867, center=<0.6805 -1.5715 4.1282>, rotation=<rot 0 -169.9244 -91.0217>}]
 dyn_coat_side_l [{shapeType=Box, extentX=2.6036, extentY=6.1498, extentZ=9.1766, center=<-0.7154 2.3264 -3.7451>, rotation=<rot 0 -171.0592 -87.6343>}]
 dyn_coat_side_r [{shapeType=Box, extentX=2.6036, extentY=5.1498, extentZ=9.6766, center=<0.5458 -2.3531 3.7451>, rotation=<rot 0 -171.0593 -87.6343>}]
 dyn_string_a_l [{shapeType=Box, extentX=0.7634, extentY=1.7949, extentZ=8.2875, center=<4.1148 -0.0752 -0.1558>, rotation=<rot 0 90.1948 -88.7638>}]
 dyn_string_a_r [{shapeType=Box, extentX=0.7634, extentY=1.7949, extentZ=7.7875, center=<-4.1935 0.0749 0.1541>, rotation=<rot 0 90.1948 91.2362>}]
 dyn_string_b_l [{shapeType=Box, extentX=1.1, extentY=2.01, extentZ=4.01, center=<2.154 0.2126 -0.0301>, rotation=<rot 0.0083 92.7747 -89.9175>}]
 dyn_string_b_r [{shapeType=Box, extentX=1.1, extentY=2.01, extentZ=4.51, center=<-2.154 -0.2126 0.0301>, rotation=<rot 0.0083 92.7747 90.0825>}]
 dyn_string_base_l [{shapeType=Capsule, length=0.505, radius=0.505, rotation=<rot 0 -90 0>}]
 dyn_string_base_r [{shapeType=Capsule, length=0.505, radius=0.505, rotation=<rot 0 -90 0>}]
 dyn_string_c_l [{shapeType=Box, extentX=0.7893, extentY=2.3222, extentZ=5.8795, center=<2.8989 -0.0199 -0.0224>, rotation=<rot 0 -96.0629 -89.4847>}]
 dyn_string_c_r [{shapeType=Box, extentX=0.7893, extentY=2.3222, extentZ=5.8795, center=<-2.8989 0.0199 0.0224>, rotation=<rot 0 -96.0629 90.5153>}]
 spine_02 [{shapeType=Capsule, length=3.5835, radius=7.911, center=<2.3023 1.5825 -0.1041>, rotation=<rot -0.2549 -89.9982 89.7738>}]
 spine_04 [{shapeType=Capsule, length=9.5484, radius=10.3689, center=<-0.7629 2.1614 0.0014>, rotation=<rot 0.0287 -90 0.023>}]
}
constraints [
 {bone1Name=dyn_string_a_l, bone2Name=dyn_string_base_l, swing1LimitDegrees=25, swing2LimitDegrees=30, twistLimitDegrees=1}
 {bone1Name=dyn_string_b_l, bone2Name=dyn_string_a_l, swing1LimitDegrees=30, swing2LimitDegrees=30, twistLimitDegrees=1}
 {bone1Name=dyn_string_c_l, bone2Name=dyn_string_b_l, swing1LimitDegrees=15, swing2LimitDegrees=15, twistLimitDegrees=1}
 {bone1Name=dyn_string_a_r, bone2Name=dyn_string_base_r, swing1LimitDegrees=25, swing2LimitDegrees=30, twistLimitDegrees=1}
 {bone1Name=dyn_string_b_r, bone2Name=dyn_string_a_r, swing1LimitDegrees=30, swing2LimitDegrees=30, twistLimitDegrees=1}
 {bone1Name=dyn_string_c_r, bone2Name=dyn_string_b_r, swing1LimitDegrees=15, swing2LimitDegrees=15, twistLimitDegrees=1}
 {
  bone1Name=dyn_string_base_l
  bone2Name=spine_04
  swing1LimitDegrees=45
  swing1Motion=Free
  swing2LimitDegrees=45
  swing2Motion=Free
  twistLimitDegrees=45
  twistMotion=Free
 }
 {
  bone1Name=dyn_string_base_r
  bone2Name=spine_04
  swing1LimitDegrees=45
  swing1Motion=Free
  swing2LimitDegrees=45
  swing2Motion=Free
  twistLimitDegrees=45
  twistMotion=Free
 }
 {bone1Name=dyn_coat_fr_l, bone2Name=spine_04, swing1LimitDegrees=60, swing2LimitDegrees=60, twistLimitDegrees=20}
 {bone1Name=dyn_coat_side_l, bone2Name=spine_04, swing1LimitDegrees=60, swing2LimitDegrees=60, twistLimitDegrees=20}
 {bone1Name=dyn_coat_fr_r, bone2Name=spine_04, swing1LimitDegrees=60, swing2LimitDegrees=60, twistLimitDegrees=20}
 {bone1Name=dyn_coat_side_r, bone2Name=spine_04, swing1LimitDegrees=60, swing2LimitDegrees=60, twistLimitDegrees=20}
 {bone1Name=dyn_coat_bk, bone2Name=spine_04, swing1LimitDegrees=60, swing2LimitDegrees=60, twistLimitDegrees=20}
 {
  bone1Name=spine_04
  bone2Name=spine_02
  swing1LimitDegrees=45
  swing1Motion=Free
  swing2LimitDegrees=45
  swing2Motion=Free
  twistLimitDegrees=45
  twistMotion=Free
 }
 {
  bone1Name=dyn_coat_bk
  bone2Name=dyn_coat_side_l
  swing1LimitDegrees=45
  swing1Motion=Locked
  swing2LimitDegrees=45
  swing2Motion=Locked
  twistLimitDegrees=45
  twistMotion=Locked
 }
 {
  bone1Name=dyn_coat_bk
  bone2Name=dyn_coat_side_r
  swing1LimitDegrees=45
  swing1Motion=Locked
  swing2LimitDegrees=45
  swing2Motion=Locked
  twistLimitDegrees=45
  twistMotion=Locked
 }
]
