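# Secondary-motion AnimBlueprint for the F_MED_Reverb player body, layering corrective and dynamic bone motion onto the attached parent mesh's copied pose.
# The EventGraph measures head, arm, shoulder and thigh deltas from ref pose, driving thigh-volume and collar correctives alongside fixed bone constraints.
# Sine and Perlin noise from speakerReverb_FNC pulse the speaker bones' translation and scale.
# A final RigidBody node simulates the dynamics physics asset relative to spine_02.
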AnimBlueprint /Game/Characters/Player/Female/Medium/Bodies/F_MED_Reverb/Meshes/F_MED_Reverb_AnimBP
target_skeleton=&/F_MED_Reverb_Skeleton
graphs [
 AnimGraph {nodes [
   n1 = "Constraint - thigh_volume_l (R)"(ComponentPose=Local To Component(LocalPose=Copy Pose From Mesh() {bUseAttachedParent=true}), Alpha=0.5, ConstraintWeights_0=1) {
    alphaBoolBlend {blendOption=Linear}
    boneToModify {boneName=thigh_volume_l}
    constraintSetup [{transformType=Rotation, targetBone {boneName=thigh_l}}]
    constraintWeights [1]
   }
   n2 = "Constraint - thigh_volume_r (R)"(ComponentPose=@n1, Alpha=0.5, ConstraintWeights_0=1) {
    alphaBoolBlend {blendOption=Linear}
    boneToModify {boneName=thigh_volume_r}
    constraintSetup [{transformType=Rotation, targetBone {boneName=thigh_r}}]
    constraintWeights [1]
   }
   n3 = "Transform (Modify) Bone - Bone: thigh_volume_l"(ComponentPose=@n2, Alpha=Get thigh_rot_l(), Translation="-1,-1,1", Rotation="0,0,0", Scale="1,1,1") {
    translationMode=BMM_Additive
    translationSpace=BCS_BoneSpace
    alphaBoolBlend {blendOption=Linear}
    boneToModify {boneName=thigh_volume_l}
    translation=<-1 -1 1>
   }
   n4 = "Transform (Modify) Bone - Bone: thigh_volume_r"(ComponentPose=@n3, Alpha=Get thigh_rot_r(), Translation="1,1,-1", Rotation="0,0,0", Scale="1,1,1") {
    translationMode=BMM_Additive
    translationSpace=BCS_BoneSpace
    alphaBoolBlend {blendOption=Linear}
    boneToModify {boneName=thigh_volume_r}
    translation=<1 1 -1>
   }
   n5 = Get headFwdBk() -> (headFwdBk:double)
   n6 = "Transform (Modify) Bone - Bone: dyn_collar_bk_base"(ComponentPose=@n4, Alpha=@n5, Translation="0,0,0", Rotation="-40,0,0", Scale="1,1,0.6") {
    rotationMode=BMM_Additive
    rotationSpace=BCS_BoneSpace
    scaleMode=BMM_Additive
    scaleSpace=BCS_BoneSpace
    translationMode=BMM_Additive
    translationSpace=BCS_BoneSpace
    alphaBoolBlend {blendOption=Linear}
    alphaScaleBiasClamp {bMapRange=true, inRange {max=0, min=-45}, outRange {max=0, min=1}}
    boneToModify {boneName=dyn_collar_bk_base}
    rotation {pitch=-40}
    scale {z=0.6}
   }
   n7 = "Transform (Modify) Bone - Bone: dyn_collar_side_base_l"(ComponentPose=@n6, Alpha=@n5, Translation="0,0,0", Rotation="0,0,-20", Scale="1,1,0.75") {
    rotationMode=BMM_Additive
    rotationSpace=BCS_BoneSpace
    scaleMode=BMM_Additive
    scaleSpace=BCS_BoneSpace
    translationMode=BMM_Additive
    translationSpace=BCS_BoneSpace
    alphaBoolBlend {blendOption=Linear}
    alphaScaleBiasClamp {bMapRange=true, inRange {max=0, min=-45}, outRange {max=0, min=1}}
    boneToModify {boneName=dyn_collar_side_base_l}
    rotation {roll=-20}
    scale {z=0.75}
   }
   n8 = "Transform (Modify) Bone - Bone: dyn_collar_side_base_r"(ComponentPose=@n7, Alpha=@n5, Translation="0,0,0", Rotation="0,0,-20", Scale="1,1,0.75") {
    rotationMode=BMM_Additive
    rotationSpace=BCS_BoneSpace
    scaleMode=BMM_Additive
    scaleSpace=BCS_BoneSpace
    translationMode=BMM_Additive
    translationSpace=BCS_BoneSpace
    alphaBoolBlend {blendOption=Linear}
    alphaScaleBiasClamp {bMapRange=true, inRange {max=0, min=-45}, outRange {max=0, min=1}}
    boneToModify {boneName=dyn_collar_side_base_r}
    rotation {roll=-20}
    scale {z=0.75}
   }
   n9 = Get armUp_L() -> (armUp_L:double)
   n10 = "Transform (Modify) Bone - Bone: dyn_collar_fr_base_l"(ComponentPose=@n8, Alpha=@n9, Translation="0,0,0", Rotation="10,-25,0", Scale="1,1,1") {
    rotationMode=BMM_Additive
    rotationSpace=BCS_BoneSpace
    scaleMode=BMM_Additive
    scaleSpace=BCS_BoneSpace
    translationMode=BMM_Additive
    translationSpace=BCS_BoneSpace
    alphaBoolBlend {blendOption=Linear}
    alphaScaleBiasClamp {bMapRange=true, inRange {max=35, min=15}}
    boneToModify {boneName=dyn_collar_fr_base_l}
    rotation {pitch=10, yaw=-25}
   }
   n11 = "Transform (Modify) Bone - Bone: dyn_collar_side_base_l"(ComponentPose=@n10, Alpha=@n9, Translation="0,0,0", Rotation="15,10,0", Scale="1,1,1") {
    rotationMode=BMM_Additive
    rotationSpace=BCS_BoneSpace
    scaleMode=BMM_Additive
    scaleSpace=BCS_BoneSpace
    translationMode=BMM_Additive
    translationSpace=BCS_BoneSpace
    alphaBoolBlend {blendOption=Linear}
    alphaScaleBiasClamp {bMapRange=true, inRange {max=35, min=15}}
    boneToModify {boneName=dyn_collar_side_base_l}
    rotation {pitch=15, yaw=10}
   }
   n12 = Get shoulderUp_R() -> (shoulderUp_R:double)
   n13 = "Transform (Modify) Bone - Bone: dyn_collar_fr_base_r"(ComponentPose=@n11, Alpha=@n12, Translation="2,0,0", Rotation="10,-10,0", Scale="1,1,1") {
    rotationMode=BMM_Additive
    rotationSpace=BCS_BoneSpace
    scaleMode=BMM_Additive
    scaleSpace=BCS_BoneSpace
    translationMode=BMM_Additive
    translationSpace=BCS_BoneSpace
    alphaBoolBlend {blendOption=Linear}
    alphaScaleBiasClamp {bMapRange=true, inRange {max=25, min=10}}
    boneToModify {boneName=dyn_collar_fr_base_r}
    rotation {pitch=10, yaw=-10}
    translation {x=2}
   }
   n14 = "Transform (Modify) Bone - Bone: dyn_collar_side_base_r"(ComponentPose=@n13, Alpha=@n12, Translation="2,0,0", Rotation="5,5,0", Scale="1,1,1") {
    rotationMode=BMM_Additive
    rotationSpace=BCS_BoneSpace
    scaleMode=BMM_Additive
    scaleSpace=BCS_BoneSpace
    translationMode=BMM_Additive
    translationSpace=BCS_BoneSpace
    alphaBoolBlend {blendOption=Linear}
    alphaScaleBiasClamp {bMapRange=true, inRange {max=25, min=10}}
    boneToModify {boneName=dyn_collar_side_base_r}
    rotation {pitch=5, yaw=5}
    translation {x=2}
   }
   n15 = "Constraint - dyn_string_base_l (R)"(ComponentPose=@n14, Alpha=1, ConstraintWeights_0=1) {
    alphaBoolBlend {blendOption=Linear}
    boneToModify {boneName=dyn_string_base_l}
    constraintSetup [{transformType=Rotation, targetBone {boneName=spine_04}}]
    constraintWeights [1]
   }
   n16 = "Constraint - dyn_string_base_r (R)"(ComponentPose=@n15, Alpha=1, ConstraintWeights_0=1) {
    alphaBoolBlend {blendOption=Linear}
    boneToModify {boneName=dyn_string_base_r}
    constraintSetup [{transformType=Rotation, targetBone {boneName=spine_04}}]
    constraintWeights [1]
   }
   n17 = "Constraint - dyn_string_b_l (P)"(ComponentPose=@n16, Alpha=0.5, ConstraintWeights_0=1) {
    alphaBoolBlend {blendOption=Linear}
    boneToModify {boneName=dyn_string_b_l}
    constraintSetup [{targetBone {boneName=spine_04}}]
    constraintWeights [1]
   }
   n18 = "Constraint - dyn_string_b_r (P)"(ComponentPose=@n17, Alpha=0.5, ConstraintWeights_0=1) {
    alphaBoolBlend {blendOption=Linear}
    boneToModify {boneName=dyn_string_b_r}
    constraintSetup [{targetBone {boneName=spine_04}}]
    constraintWeights [1]
   }
   n19 = "Constraint - dyn_leg_speaker_l (P)"(ComponentPose=@n18, Alpha=0.5, ConstraintWeights_0=1) {
    alphaBoolBlend {blendOption=Linear}
    boneToModify {boneName=dyn_leg_speaker_l}
    constraintSetup [{targetBone {boneName=thigh_twist_01_l}}]
    constraintWeights [1]
   }
   n20 = "Constraint - dyn_leg_speaker_r (P)"(ComponentPose=@n19, Alpha=0.5, ConstraintWeights_0=1) {
    alphaBoolBlend {blendOption=Linear}
    boneToModify {boneName=dyn_leg_speaker_r}
    constraintSetup [{targetBone {boneName=thigh_twist_01_r}}]
    constraintWeights [1]
   }
   n21 = Get reverbScale() -> (reverbScale:Vector)
   n22 = Get reverbTran() -> (reverbTran:Vector)
   n23 = "Transform (Modify) Bone - Bone: dyn_uparm_speaker_l"(ComponentPose=@n20, Alpha=1, Translation=@n22, Rotation="0,0,0", Scale=@n21) {
    rotationMode=BMM_Additive
    rotationSpace=BCS_BoneSpace
    scaleMode=BMM_Additive
    scaleSpace=BCS_BoneSpace
    translationMode=BMM_Additive
    translationSpace=BCS_BoneSpace
    alphaBoolBlend {blendOption=Linear}
    alphaScaleBiasClamp {inRange {max=25, min=10}}
    boneToModify {boneName=dyn_uparm_speaker_l}
   }
   n24 = "Transform (Modify) Bone - Bone: dyn_uparm_speaker_r"(ComponentPose=@n23, Alpha=1, Translation=@n22, Rotation="0,0,0", Scale=@n21) {
    rotationMode=BMM_Additive
    rotationSpace=BCS_BoneSpace
    scaleMode=BMM_Additive
    scaleSpace=BCS_BoneSpace
    translationMode=BMM_Additive
    translationSpace=BCS_BoneSpace
    alphaBoolBlend {blendOption=Linear}
    alphaScaleBiasClamp {inRange {max=25, min=10}}
    boneToModify {boneName=dyn_uparm_speaker_r}
   }
   n25 = "Transform (Modify) Bone - Bone: dyn_neck_speaker_l"(ComponentPose=@n24, Alpha=1, Translation=@n22, Rotation="0,0,0", Scale=@n21) {
    rotationMode=BMM_Additive
    rotationSpace=BCS_BoneSpace
    scaleMode=BMM_Additive
    scaleSpace=BCS_BoneSpace
    translationMode=BMM_Additive
    translationSpace=BCS_BoneSpace
    alphaBoolBlend {blendOption=Linear}
    alphaScaleBiasClamp {inRange {max=25, min=10}}
    boneToModify {boneName=dyn_neck_speaker_l}
   }
   n26 = "Transform (Modify) Bone - Bone: dyn_neck_speaker_r"(ComponentPose=@n25, Alpha=1, Translation=@n22, Rotation="0,0,0", Scale=@n21) {
    rotationMode=BMM_Additive
    rotationSpace=BCS_BoneSpace
    scaleMode=BMM_Additive
    scaleSpace=BCS_BoneSpace
    translationMode=BMM_Additive
    translationSpace=BCS_BoneSpace
    alphaBoolBlend {blendOption=Linear}
    alphaScaleBiasClamp {inRange {max=25, min=10}}
    boneToModify {boneName=dyn_neck_speaker_r}
   }
   n27 = "Transform (Modify) Bone - Bone: dyn_leg_speaker_l"(ComponentPose=@n26, Alpha=1, Translation=@n22, Rotation="0,0,0", Scale=@n21) {
    rotationMode=BMM_Additive
    rotationSpace=BCS_BoneSpace
    scaleMode=BMM_Additive
    scaleSpace=BCS_BoneSpace
    translationMode=BMM_Additive
    translationSpace=BCS_BoneSpace
    alphaBoolBlend {blendOption=Linear}
    alphaScaleBiasClamp {inRange {max=25, min=10}}
    boneToModify {boneName=dyn_leg_speaker_l}
   }
   n28 = "Transform (Modify) Bone - Bone: dyn_leg_speaker_r"(ComponentPose=@n27, Alpha=1, Translation=@n22, Rotation="0,0,0", Scale=@n21) {
    rotationMode=BMM_Additive
    rotationSpace=BCS_BoneSpace
    scaleMode=BMM_Additive
    scaleSpace=BCS_BoneSpace
    translationMode=BMM_Additive
    translationSpace=BCS_BoneSpace
    alphaBoolBlend {blendOption=Linear}
    alphaScaleBiasClamp {inRange {max=25, min=10}}
    boneToModify {boneName=dyn_leg_speaker_r}
   }
   n29 = RigidBody(ComponentPose=@n28) {
    bClampLinearTranslationLimitToRefPose=true
    bOverrideWorldGravity=true
    overridePhysicsAsset=&/F_MED_Reverb_Dyn_Physics
    simulationSpace=BaseBoneSpace
    alphaBoolBlend {blendOption=Linear}
    baseBoneRef {boneName=spine_02}
    componentLinearAccScale=<0.3 0.3 0.3>
    componentLinearVelScale=<0.3 0.3 0.3>
    overrideWorldGravity {z=-980}
   }
   Output Pose(Result=Component To Local(ComponentPose=@n29)) {layerGroup=None}
  ]}
 EventGraph {nodes [
   Event BlueprintUpdateAnimation()
   > n1 = Sequence!(execute) -> (then_0, then_1, then_2:Exec, then_3:Exec, then_4:Exec, then_5:Exec, then_6:Exec)
   n2 = GetOwnerSkeletalMeshComponent()
   n3 = GetDeltaTransformFromRefPose(self=@n2, BoneName=head, BaseName=spine_05)
   > Set headFwdBk(headFwdBk:double=BreakRotator(InRot=BreakTransform(InTransform=@n3).Rotation).Yaw) -> (Output_Get:double)
   n4 = BreakTransform(InTransform=GetDeltaTransformFromRefPose(self=@n2, BoneName=thigh_r, BaseName=pelvis))
   Set thigh_rot_r(execute=@n1.then_4, thigh_rot_r=MapRangeClamped(Value=BreakRotator(InRot=@n4.Rotation).Yaw, InRangeA=0, InRangeB=-105))
   n5 = GetDeltaTransformFromRefPose(self=@n2, BoneName=upperarm_l, BaseName=clavicle_l)
   Set armUp_L(execute=@n1.then_1, armUp_L:double=BreakRotator(InRot=BreakTransform(InTransform=@n5).Rotation).Pitch) -> (Output_Get:double)
   n6 = GetDeltaTransformFromRefPose(self=@n2, BoneName=upperarm_r, BaseName=clavicle_r)
   Set armUp_R(execute=@n1.then_2, armUp_R:double=BreakRotator(InRot=BreakTransform(InTransform=@n6).Rotation).Pitch) -> (Output_Get:double)
   n7 = GetDeltaTransformFromRefPose(self=@n2, BoneName=clavicle_r, BaseName=spine_05)
   Set shoulderUp_R(execute=@n1.then_3, shoulderUp_R:double=BreakRotator(InRot=BreakTransform(InTransform=@n7).Rotation).Pitch) -> (Output_Get:double)
   n8 = BreakTransform(InTransform=GetDeltaTransformFromRefPose(self=@n2, BoneName=thigh_l, BaseName=pelvis))
   Set thigh_rot_l(execute=@n1.then_5, thigh_rot_l=MapRangeClamped(Value=BreakRotator(InRot=@n8.Rotation).Yaw, InRangeA=-20, InRangeB=-90))
   n9 = speakerReverb_FNC!(execute=@n1.then_6, self, frequency:double=5, amp:double=0.2) -> (then, tran_output:double, scale_output:double)
   > Set reverbTran(reverbTran:Vector=MakeVector(X=@n9, Y=0)) -> (Output_Get:Vector)
   > Set reverbScale(reverbScale:Vector=MakeVector(X=@n9.scale_output, Y=@n9.scale_output, Z=@n9.scale_output)) -> (Output_Get:Vector)
   n10 = Branch(Condition=Get bIsJumping())
   > Delay(Duration=0.25)
   > Set jacket_ext_force(jacket_ext_force:Vector="0,0,1250") -> (Output_Get:Vector)
   > n11 = Delay(Duration=0.375)
   Do Once(execute=@n10.else, Reset=@n11.then)
   > Set jacket_ext_force(jacket_ext_force="0, 0, 0")
   GetGameTimeInSeconds()
   TryGetPawnOwner()
  ]}
 speakerReverb_FNC {nodes [
   n1 = speakerReverb_FNC!() -> (then, frequency:double=1, amp:double=1) {is=title K2Node_FunctionEntry}
   n2 = GetGameTimeInSeconds()
   n3 = "Sin (Radians)"(A=PerlinNoise1D!(Value:float=float * float(A=@n1, B=@n2)) -> (ReturnValue:float))
   > Return Node!(
    execute
    tran_output:double=PerlinNoise1D(Value=float * float(A=@n3, B=@n1.amp))
    scale_output:double=Make Perlin Noise and Remap!(Value:float=@n3, RangeOutMin:float=0.9, RangeOutMax:float=1.1) -> (ReturnValue:float)
   )
  ]}
]
variables [
 armUp_L {type=number}
 armUp_R {type=number}
 headFwdBk {type=number}
 jacket_ext_force {type=Vector}
 reverbScale {type=Vector}
 reverbTran {type=Vector}
 shoulderUp_R {type=number}
 thigh_rot_l {type=number}
 thigh_rot_r {type=number}
]
