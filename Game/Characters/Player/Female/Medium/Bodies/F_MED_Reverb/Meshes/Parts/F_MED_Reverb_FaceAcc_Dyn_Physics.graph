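# Physics asset for the Reverb female medium character's dynamic face accessory mesh, defining capsule bodies and joint limits for hair bones.
# A large head capsule anchors every strand, either as two-segment chains or single bones, including five front pieces.
# Chain roots constrain to the head with tight swing, while tips constrain to their roots with wider swing.
# Twist stays narrow throughout, and most front pieces use asymmetric swing limits to favor one axis.
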PhysicsAsset /Game/Characters/Player/Female/Medium/Bodies/F_MED_Reverb/Meshes/Parts/F_MED_Reverb_FaceAcc_Dyn_Physics
bodies {
 dyn_hair_a_1 [{shapeType=Capsule, length=1.9095, radius=1.6931, center=<1.9763 -0.1859 -0.1298>, rotation=<rot 0 92.608 -95.6368>}]
 dyn_hair_a_2 [{shapeType=Capsule, length=2.587, radius=1.7457, center=<2.5423 -0.1893 -0.3664>, rotation=<rot 0 -88.5461 -73.6755>}]
 dyn_hair_b_1 [{shapeType=Capsule, length=2.2003, radius=1.8786, center=<2.5101 -0.0436 -0.2528>, rotation=<rot 2.9733 96.357 -100.2404>}]
 dyn_hair_b_2 [{shapeType=Capsule, length=3.0365, radius=1.639, center=<2.9319 -0.1082 0.0628>, rotation=<rot 0 -86.7275 -89.3796>}]
 dyn_hair_c_1 [{shapeType=Capsule, length=3.2108, radius=1.6458, center=<2.7477 0.5182 -0.2907>, rotation=<rot 0 -79.1389 -82.4253>}]
 dyn_hair_d_1 [{shapeType=Capsule, length=2.1247, radius=1.5802, center=<1.7811 0.0494 -0.3284>, rotation=<rot 0 -91.543 -74.9867>}]
 dyn_hair_d_2 [{shapeType=Capsule, length=2.9598, radius=1.7382, center=<2.5506 -0.0018 -0.1945>, rotation=<rot 0 -86.1986 -78.9391>}]
 dyn_hair_e_1 [{shapeType=Capsule, length=2.7573, radius=1.5898, center=<1.6966 -0.0765 -0.154>, rotation=<rot 7.9319 77.2536 -112.8817>}]
 dyn_hair_e_2 [{shapeType=Capsule, length=3.0061, radius=1.9193, center=<2.5407 -0.0013 -0.3632>, rotation=<rot 0 -91.8894 -77.222>}]
 dyn_hair_f_1 [{shapeType=Capsule, length=2.4633, radius=1.6252, center=<2.1343 -0.1149 -0.315>, rotation=<rot 0 -90.789 -85.5574>}]
 dyn_hair_f_2 [{shapeType=Capsule, length=3.8581, radius=1.8402, center=<3.1979 -0.081 -0.8072>, rotation=<rot 0 -89.2116 -68.6243>}]
 dyn_hair_fr [{shapeType=Capsule, length=2.25, radius=1.5, center=<0.0175 0.0265 -0.8007>}]
 dyn_hair_fr_a_l [{shapeType=Capsule, length=1.9735, radius=0.75, center=<1.326 0.0893 -0.3029>, rotation=<rot 0 -83.0286 -80.409>}]
 dyn_hair_fr_a_r [{shapeType=Capsule, length=2.255, radius=0.75, center=<-0.9648 0.0072 0.4881>, rotation=<rot 0 -87.7831 -81.6724>}]
 dyn_hair_fr_l [{shapeType=Capsule, length=2.25, radius=1.5, center=<0 0 -0.8013>}]
 dyn_hair_fr_r [{shapeType=Capsule, length=2.25, radius=1.5, center=<-0.0241 0.0472 0.7995>}]
 dyn_hair_g_1 [{shapeType=Capsule, length=1.4216, radius=1.4842, center=<0.8901 -0.0448 -0.0426>, rotation=<rot -37.1976 -90 -82.7871>}]
 dyn_hair_g_2 [{shapeType=Capsule, length=3.4055, radius=1.592, center=<3.0197 0.1765 -0.2413>, rotation=<rot 0 -85.8362 -82.2322>}]
 dyn_hair_i_1 [{shapeType=Capsule, length=3.4342, radius=1.561, center=<2.8742 0.0472 -0.4532>, rotation=<rot 0 -89.979 -74.084>}]
 head [{shapeType=Capsule, length=4.25, radius=5, rotation=<rot 90 0 0>}]
}
constraints [
 {bone1Name=dyn_hair_g_1, bone2Name=head, swing1LimitDegrees=15, swing2LimitDegrees=15, twistLimitDegrees=5}
 {bone1Name=dyn_hair_g_2, bone2Name=dyn_hair_g_1, swing1LimitDegrees=30, swing2LimitDegrees=30, twistLimitDegrees=5}
 {bone1Name=dyn_hair_i_1, bone2Name=head, swing1LimitDegrees=30, swing2LimitDegrees=40, twistLimitDegrees=5}
 {bone1Name=dyn_hair_e_1, bone2Name=head, swing1LimitDegrees=15, swing2LimitDegrees=15, twistLimitDegrees=5}
 {bone1Name=dyn_hair_e_2, bone2Name=dyn_hair_e_1, swing1LimitDegrees=30, swing2LimitDegrees=30, twistLimitDegrees=5}
 {bone1Name=dyn_hair_f_1, bone2Name=head, swing1LimitDegrees=15, swing2LimitDegrees=15, twistLimitDegrees=5}
 {bone1Name=dyn_hair_f_2, bone2Name=dyn_hair_f_1, swing1LimitDegrees=30, swing2LimitDegrees=30, twistLimitDegrees=5}
 {bone1Name=dyn_hair_d_1, bone2Name=head, swing1LimitDegrees=15, swing2LimitDegrees=15, twistLimitDegrees=5}
 {bone1Name=dyn_hair_d_2, bone2Name=dyn_hair_d_1, swing1LimitDegrees=30, swing2LimitDegrees=30, twistLimitDegrees=5}
 {bone1Name=dyn_hair_c_1, bone2Name=head, swing1LimitDegrees=30, swing2LimitDegrees=30, twistLimitDegrees=5}
 {bone1Name=dyn_hair_b_1, bone2Name=head, swing1LimitDegrees=15, swing2LimitDegrees=15, twistLimitDegrees=5}
 {bone1Name=dyn_hair_b_2, bone2Name=dyn_hair_b_1, swing1LimitDegrees=30, swing2LimitDegrees=30, twistLimitDegrees=5}
 {bone1Name=dyn_hair_a_1, bone2Name=head, swing1LimitDegrees=10, swing2LimitDegrees=10, twistLimitDegrees=5}
 {bone1Name=dyn_hair_a_2, bone2Name=dyn_hair_a_1, swing1LimitDegrees=20, swing2LimitDegrees=20, twistLimitDegrees=5}
 {bone1Name=dyn_hair_fr, bone2Name=head, swing1LimitDegrees=5, swing2LimitDegrees=30, twistLimitDegrees=10}
 {bone1Name=dyn_hair_fr_l, bone2Name=head, swing1LimitDegrees=5, swing2LimitDegrees=25, twistLimitDegrees=15}
 {bone1Name=dyn_hair_fr_r, bone2Name=head, swing1LimitDegrees=5, swing2LimitDegrees=20, twistLimitDegrees=15}
 {bone1Name=dyn_hair_fr_a_l, bone2Name=head, swing1LimitDegrees=15, swing2LimitDegrees=5, twistLimitDegrees=1}
 {bone1Name=dyn_hair_fr_a_r, bone2Name=head, swing1LimitDegrees=15, swing2LimitDegrees=5, twistLimitDegrees=1}
]
